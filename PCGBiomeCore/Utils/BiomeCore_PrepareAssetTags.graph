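# Editor utility that writes mesh render settings into PCG_ actor tags on every StaticMeshActor in the editor world, preparing assets for Biome Core.
# Its custom event rebuilds each actor's tags, keeping those without the pcg_ prefix so old generated tags are replaced.
# It adds flag tags for shadows, visibility, HLOD inclusion and distance field lighting.
# Detail mode and WPO disable distance are encoded as name:value tags.
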EditorUtilityBlueprint /PCGBiomeCore/Utils/BiomeCore_PrepareAssetTags
graphs [EventGraph {nodes [
   BiomeCore: PrepareAssetTags() {is=title K2Node_CustomEvent}
   > n1 = GetEditorWorld(self=Get UnrealEditorSubsystem())
   > n2 = GetAllActorsOfClass!(execute, WorldContextObject:Object Reference=@n1, ActorClass:Actor Class Reference=&StaticMeshActor) -> (then, OutActors:Array of Static Mesh Actor Object References)
   > n3 = For Each Loop(Array:Array of Static Mesh Actor Object References=@n2) -> (Array Element:Static Mesh Actor Object Reference)
   > n4 = Sequence!(execute) -> (then_0, then_1, then_2:Exec, then_3:Exec, then_4:Exec, then_5:Exec, then_6:Exec, then_7:Exec, then_8:Exec, then_9:Exec)
   > Clear(TargetArray:Array of Names=Get UserTags() -> (UserTags:Array of Names))
   > n5 = For Each Loop(Array:Array of Names=Get Tags!(self:Actor Object Reference=@n3) -> (Tags:Array of Names)) -> (Array Element:Name)
   > n6 = Branch(Condition=StartsWith(SourceString="To String (Name)"(InName=@n5), InPrefix=pcg_))
   Add(execute=@n6.else, TargetArray:Array of Names=Get UserTags(), NewItem:"Name (by ref)"=@n5) {is=title K2Node_CallArrayFunction}
   n7 = Get StaticMeshComponent(self=@n3)
   Branch(execute=@n4.then_1, Condition=Get CastShadow!(self:Primitive Component Object Reference=@n7) -> (CastShadow:bool))
   > Add(TargetArray=Get UserTags(), NewItem=PCG_CastShadow)
   Branch(execute=@n4.then_2, Condition=Get bCastHiddenShadow!(self:Primitive Component Object Reference=@n7) -> (bCastHiddenShadow:bool))
   > Add(TargetArray=Get UserTags(), NewItem=PCG_HiddenShadow)
   Branch(execute=@n4.then_3, Condition=Get bVisible!(self:Scene Component Object Reference=@n7) -> (bVisible:bool))
   > Add(TargetArray=Get UserTags(), NewItem=PCG_Visible)
   Branch(
    execute=@n4.then_4
    Condition=AND Boolean(
     A=Get bEnableAutoLODGeneration!(self:Actor Object Reference=@n3) -> (bEnableAutoLODGeneration:bool)
     B=Get bEnableAutoLODGeneration(self:Primitive Component Object Reference=@n7)
    )
   )
   > Add(TargetArray=Get UserTags(), NewItem=PCG_IncludeInHLOD)
   Branch(
    execute=@n4.then_5
    Condition=Get bAffectDistanceFieldLighting!(self:Primitive Component Object Reference=@n7) -> (bAffectDistanceFieldLighting:bool)
   )
   > Add(TargetArray=Get UserTags(), NewItem=PCG_AffectDistanceFieldLighting)
   n8 = "To String (Integer)"(InInt="To Integer (Byte)"!(InByte:Byte=Get DetailMode!(self:Scene Component Object Reference=@n7) -> (DetailMode:EDetailMode Enum=DM_Low)) -> (ReturnValue:int=0))
   Add(execute=@n4.then_6, TargetArray=Get UserTags(), NewItem=String To Name(InString=Append(A=PCG_DetailMode:, B=@n8)))
   n9 = "To String (Integer)"(InInt=Get WorldPositionOffsetDisableDistance!(self:Static Mesh Component Object Reference=@n7) -> (WorldPositionOffsetDisableDistance:int=0))
   Add(execute=@n4.then_7, TargetArray=Get UserTags(), NewItem=String To Name(InString=Append(A=PCG_WPODisableDistance:, B=@n9)))
   Set Tags!(execute=@n4.then_9, Tags:Array of Names=Get UserTags(), self:Actor Object Reference=@n3) -> (then, Output_Get:Array of Names)
  ]}]
variables [UserTags {type="array<string>", default []}]
